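; Gameplay ability for a PlotTwist quest that, going by its name, tracks the player Force Jumping onto a grind rail.
; On activation it starts two independent tag waits, one for the Force Jump power tag and one for the grinding movement tag.
; They apply the DoubleJump and RecentlyGrinded effects to the owner respectively.
; The ability itself never checks the order or combination of the two events.
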(event Ability|EventActivateAbility
  ;[1 1 1] Check if player has Force Jumped
  (|CallParentFunction)
  (Ability|Tasks|WaitGameplayTagAdd "(TagName=\"Granted.Athena.PlotTwist.Force.UsingPower.Jump\")"
    (:then
      (Ability|Tasks|WaitGameplayTagAdd "(TagName=\"MovementMode.Grinding\")"
        (:Added
          ;[1 1 1] Tick at close random intervals for perf, do not go under 1 second
          (Ability|ApplyGameplayEffectToOwner "/PlotTwistQuests/Items/QuestAbilities/DoubleJump/GE_Quest_RecentlyGrinded.GE_Quest_RecentlyGrinded_C"))))
    (:Added
      (Ability|ApplyGameplayEffectToOwner "/PlotTwistQuests/Items/QuestAbilities/DoubleJump/GE_Quest_DoubleJump.GE_Quest_DoubleJump_C"))))
